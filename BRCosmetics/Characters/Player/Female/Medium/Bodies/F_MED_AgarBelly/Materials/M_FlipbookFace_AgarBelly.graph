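# Character face material for AgarBelly, layering MF_FlipbookFace and a custom anime eye animation between the 2023 character parent PreFX and PostFX functions.
# An AgarbellyEmote parameter blends in a looping eye flipbook, with separately placed left and right UVs.
# Vertex-color masks and eye pose thresholds override World Position Offset around the eyelashes per side.
# AnimBP-driven HasPoseChanged adjusts temporal responsiveness on masked face regions.
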Material /BRCosmetics/Characters/Player/Female/Medium/Bodies/F_MED_AgarBelly/Materials/M_FlipbookFace_AgarBelly
graph {
 properties {
  bUseMaterialAttributes=true
  bUsedWithMeshDeformer=true
  bUsedWithMorphTargets=true
  bUsedWithSkeletalMesh=true
  pixelDepthOffsetMode=PDOM_Legacy
  shadingModel=MSM_FromMaterialExpression
  usedShadingModels=Default Lit | Subsurface | Clear Coat | Subsurface Profile | Cloth
 }
 nodes [
  n1 = MaterialFunctionCall() {materialFunction=&MF_CharacterParent_2023_PreFX}
  n2 = MaterialFunctionCall!(Mouth UV, Eye UV, Brow UV, Material=@n1.Result) {materialFunction=/Game/Global/Materials/Character_Parent_2023/FlipbookFace_Parent_2023/MaterialFunctions/MF_FlipbookFace}
  n3 = GetMaterialAttributes(None=@n2.Material) {attributeGetTypes [#1, #2]}
  n4 = LinearInterpolate(
   A=NamedRerouteUsage()
   B=LinearInterpolate(A=NamedRerouteUsage(), Alpha=NamedRerouteUsage()) {constB=0.02}
   Alpha=ScalarParameter() {parameterName=AgarbellyEmote}
  )
  n5 = Constant()
  n6 = ScalarParameter() {group=01_EyeAnimation_Poses, parameterName=L_upperLidDown_pose, sliderMax=1, sortPriority=3}
  n7 = ScalarParameter() {group=01_EyeAnimation_Poses, parameterName=L_blink_pose, sliderMax=1, sortPriority=1}
  n8 = Subtract(A=ScalarParameter() {group=01_EyeAnimation_Poses, parameterName=L_squint_inner_pose, sliderMax=1, sortPriority=5}) {constB=0.5}
  n9 = ScalarParameter() {defaultValue=0.05, parameterName=Eyelash threshold}
  n10 = Constant()
  n11 = If(A=Max(A=Max(A=@n6, B=@n7), B=Saturate(None=@n8)), B=@n9, "A > B"=Constant() {r=1}, "A == B"=@n10, "A < B"=@n10) {constB=0.02}
  n12 = SphereMask(A=VertexColor().R, B=Constant() {r=0.56}) {attenuationRadius=0.02}
  n13 = NamedRerouteUsage()
  n14 = Max(
   A=ScalarParameter() {group=01_EyeAnimation_Poses, parameterName=R_blink_pose, sliderMax=1, sortPriority=2}
   B=ScalarParameter() {group=01_EyeAnimation_Poses, parameterName=R_squint_inner_pose, sliderMax=1, sortPriority=6}
  )
  n15 = Constant() {r=1}
  n16 = Constant()
  n17 = VertexColor()
  n18 = Constant() {r=0.56}
  n19 = LinearInterpolate(
   A=LinearInterpolate(A=@n3.World Position Offset, B=LinearInterpolate(B=@n5, Alpha=@n11), Alpha=Multiply(A=@n12, B=@n13)) {constB=0}
   B=LinearInterpolate(B=@n5, Alpha=If(A=@n14, B=@n9, "A > B"=@n15, "A == B"=@n16, "A < B"=@n16) {constB=0.02})
   Alpha=Multiply(A=SphereMask(A=@n17.R, B=@n18) {attenuationRadius=0.02}, B=NamedRerouteUsage())
  ) {constB=0}
  MaterialFunctionCall(
   Material=MaterialFunctionCall!(Material=SetMaterialAttributes!(MaterialAttributes=@n3.MaterialAttributes, Base Color=@n4, World Position Offset=@n19) {attributeSetTypes [#1, #2]}) {materialFunction=/Game/Global/Materials/Character_Parent_2023/Functions/BaseFN_CharacterOpacity_2023}.Result
   M=@n1.M
   Pre Skinned Local Position=@n1.M
   Distance Blend=@n1.M
   Fresnel=@n1.M
  ) {materialFunction=&MF_CharacterParent_2023_PostFX}
  NamedRerouteDeclaration(None=@n3.Base Color) {name=InBaseColor, nodeColor=<rgba 0 1 0.4141 1>}
  n20 = VertexColor()
  n21 = TextureSampleParameter2D() {
   group=01_EyeAnimation_Parameters
   parameterName=EyeMask_agarbelly
   samplerSource=SSM_Wrap_WorldGroupSettings
   samplerType=SAMPLERTYPE_Masks
   sortPriority=-32
   texture=&T_WhiteMask
  }
  n22 = ChannelMaskParameter(None=StaticSwitchParameter(True=AppendVector(A=@n20, B=@n20), False=@n21.RGBA) {group=01_EyeAnimation_Parameters, parameterName=UseVertexColorForEyeMask, sortPriority=-33}) {group=01_EyeAnimation_Parameters, parameterName=EyeMaskChannel, sortPriority=-32}
  n23 = Constant()
  n24 = If(
   A=DotProduct(A=Constant3Vector() {constant=<rgba 1 1 1 0>}, B=VertexColor())
   B=Constant() {r=3}
   "A > B"=@n23
   "A == B"=@n23
   "A < B"=Constant() {r=1}
  )
  NamedRerouteDeclaration(None=MaterialFunctionCall!(
    Color Lit=NamedRerouteUsage()
    Color Shaded
    Use UV2=StaticBoolParameter() {defaultValue=true, group=01_EyeAnimation_Parameters, parameterName=Use UV2 for Eye Animation, sortPriority=-30}
    Mask=StaticSwitchParameter(True=@n22.R, False=@n24) {group=01_EyeAnimation_Parameters, parameterName=UseEyeMask, sortPriority=-33}
   ) {materialFunction=&/MF_AnimeEyeAnimation_AgarBelly}.Lit) {name=OutBaseColor, nodeColor=<rgba 0 0.2266 1 1>}
  n25 = LocalPosition() {localOrigin=InstancePreSkinning}
  n26 = ShaderStageSwitch(PixelShader=VertexInterpolator(VS=@n25.XYZ).PS, VertexShader=@n25.XYZ)
  n27 = NamedRerouteDeclaration(None=Step(X=MaterialFunctionCall(Float3=@n26) {materialFunction=&BreakOutFloat3Components}.R)) {name=Right0/Left1, nodeColor=<rgba 1 0 0.1172 1>}
  NamedRerouteDeclaration(None=@n27) {name=Left, nodeColor=<rgba 1 0 0.4688 1>}
  NamedRerouteDeclaration(None=OneMinus(None=@n27)) {name=Right, nodeColor=<rgba 0 0.7891 1 1>}
  NamedRerouteDeclaration(None=@n2.BrowsVertexMask) {name=BrowVertexMask, nodeColor=<rgba 1 0 0 1>}
  NamedRerouteDeclaration(None=@n2.MouthAlpha) {name=MouthAlpha, nodeColor=<rgba 1 0.5859 0 1>}
  n28 = Time()
  n29 = ScalarParameter() {defaultValue=1, parameterName=EmoteSpeed}
  n30 = TextureObject() {
   samplerType=SAMPLERTYPE_Masks
   texture=/BRCosmetics/Characters/Player/Female/Medium/Bodies/F_MED_AgarBelly/Textures/T_AgarBelly_Head_Flipbook_Emote
  }
  n31 = ScalarParameter() {defaultValue=0.144, parameterName=Emote_right_OffsetX}
  n32 = ScalarParameter() {defaultValue=0.22, parameterName=Emote_right_OffsetY}
  n33 = MaterialFunctionCall(
   UVs=Add(A=TextureCoordinate(), B=AppendVector(A=@n31, B=@n32))
   Texture Scale=ScalarParameter() {defaultValue=0.2, parameterName=Emote_rightScale}
  ) {materialFunction=&ScaleUVsByCenter}
  n34 = TextureCoordinate()
  n35 = ComponentMask(None=@n34) {r=true}
  n36 = OneMinus(None=ComponentMask(None=@n34) {g=true})
  n37 = ScalarParameter() {defaultValue=-0.144, parameterName=Emote_Left_OffsetX}
  n38 = ScalarParameter() {defaultValue=-0.216, parameterName=Emote_Left_OffsetY}
  n39 = MaterialFunctionCall(
   UVs=Add(A=AppendVector(A=@n35, B=@n36), B=AppendVector(A=@n37, B=@n38))
   Texture Scale=ScalarParameter() {defaultValue=0.2, parameterName=Emote_LeftScale}
  ) {materialFunction=&ScaleUVsByCenter}
  n40 = MaterialFunctionCall(
   "Animation  Phase (0-1)"=Frac(None=Multiply(A=@n28, B=@n29))
   Number of Rows=Constant() {r=2}
   Number of Columns=Constant() {r=4}
   Texture=@n30
   UVs=LinearInterpolate(A=Clamp(None=@n33.UVs) {maxDefault=0.99}, B=Clamp(None=@n39.UVs), Alpha=NamedRerouteUsage())
  ) {materialFunction=&FlipBook}
  n41 = TextureSample() {samplerType=SAMPLERTYPE_Masks, texture=/BRCosmetics/Characters/Player/Female/Medium/Bodies/F_MED_AgarBelly/Textures/T_AgarBelly_Eye_Mask}
  NamedRerouteDeclaration(None=ComponentMask(None=Multiply(A=@n40.Result, B=@n41.R)) {r=true}) {name=EmoteEyeMask, nodeColor=<rgba 0.1484 1 0 1>}
  Subtract(A=ObjectPositionWS(), B=WorldPosition().XYZ)
  n42 = NamedRerouteUsage()
  n43 = NamedRerouteUsage()
  n44 = Max(A=@n22.R, B=Max(A=Max(A=@n42, B=@n43), B=NamedRerouteUsage()))
  n45 = ScalarParameter() {desc="Set by base head AnimBP if \"IsFlipbookFace\" is true", parameterName=HasPoseChanged}
  TemporalResponsivenessOutput!(None=LinearInterpolate(Alpha=Multiply(A=@n44, B=@n45)) {constA=0.4})
 ]
}
