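# Animation Blueprint for the female medium AgarBelly body cosmetic, which follows the character's animation and adds Control Rig adjustments.
# AnimGraph copies the pose from the attached parent mesh and runs it through F_MED_AgarBelly_CtrlRig before output.
# Each update sets Crouching CR to 1 when crouching without 2D movement, otherwise 0.
# The AnimGraph's getter for Crouching CR has no visible connection to the rig.
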AnimBlueprint /BRCosmetics/Characters/Player/Female/Medium/Bodies/F_MED_AgarBelly/Meshes/F_MED_AgarBelly_AnimBP
target_skeleton=&/F_MED_AgarBelly_Skeleton
graphs [
 AnimGraph {nodes [
   n1 = Control Rig(Source=Copy Pose From Mesh() {bUseAttachedParent=true}) {controlRigAssetReference {blueprintRigClass=&/F_MED_AgarBelly_CtrlRig.F_MED_AgarBelly_CtrlRig_C}}
   Output Pose(Result=@n1) {layerGroup=None}
   Get Crouching CR() -> (Crouching CR:double)
  ]}
 EventGraph {nodes [
   Event BlueprintUpdateAnimation()
   n1 = NOT Boolean(A=Get bIsMoving2D())
   n2 = Get bIsCrouching()
   > Set Crouching CR(Crouching CR:double="To Float (Boolean)"(InBool=AND Boolean(A=@n2, B=@n1))) -> (Output_Get:double)
  ]}
]
variables [Crouching CR {type=number}]
