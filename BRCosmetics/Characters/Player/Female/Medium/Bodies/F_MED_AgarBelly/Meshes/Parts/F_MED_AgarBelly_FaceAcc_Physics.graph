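# Physics asset for the AgarBelly female medium body's face accessory mesh, simulating side hair and hat pieces.
# Mirrored three-link capsule chains for hair and hat hang from small base bodies, which link to the head with free angular motion.
# Chain joints use tight twist limits, with hat swing widening toward the tips.
# A large head capsule and near-spherical arm-aim capsules, which have no constraints, add collision volumes.
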PhysicsAsset /BRCosmetics/Characters/Player/Female/Medium/Bodies/F_MED_AgarBelly/Meshes/Parts/F_MED_AgarBelly_FaceAcc_Physics
bodies {
 base_hairSide_l_01 [{shapeType=Capsule, length=0.505, radius=0.505, shapeName=base_hairSide_l_01_capsule, rotation=<rot 0 -90 0>}]
 base_hairSide_r_01 [{shapeType=Capsule, length=0.505, radius=0.505, shapeName=base_hairSide_r_01_capsule, rotation=<rot 0 -90 0>}]
 base_hat_01_l [{shapeType=Capsule, length=0.505, radius=0.505, shapeName=base_hat_01_l_capsule, rotation=<rot 0 -90 0>}]
 base_hat_01_r [{shapeType=Capsule, length=0.505, radius=0.505, shapeName=base_hat_01_r_capsule, rotation=<rot 0 -90 0>}]
 dyn_arm_aim_l [{shapeType=Capsule, length=0.0765, radius=9.4629, shapeName=dyn_arm_aim_l_capsule, center=<0.3394 -0.3521 -0.2733>, rotation=<rot 90 90 0>}]
 dyn_arm_aim_r [{shapeType=Capsule, length=0.0765, radius=9.4629, shapeName=dyn_arm_aim_l_capsule, center=<-0.3394 0.3521 0.2733>, rotation=<rot 90 90 0>}]
 dyn_hairSide_l_01 [{
   shapeType=Capsule
   length=6.8803
   radius=1.058
   shapeName=dyn_hairSide_l_01_capsule
   center=<2.0087 -1.5747 -0.007>
   rotation=<rot 0 0 -0.3997>
  }]
 dyn_hairSide_l_02 [
  {
   shapeType=Capsule
   length=7.344
   radius=1.1475
   shapeName=dyn_hairSide_l_02_capsule
   center=<2 -0.2156 0>
   rotation=<rot 0.0187 -0.0062 3.6071>
  }
  {
   shapeType=Capsule
   length=7.9637
   radius=1.1475
   shapeName=dyn_hairSide_l_02_capsule
   center=<4.4713 -0.2588 0.0657>
   rotation=<rot -3.5482 -0.4679 7.411>
  }
 ]
 dyn_hairSide_l_03 [
  {
   shapeType=Capsule
   length=9.2846
   radius=1.2313
   shapeName=dyn_hairSide_l_03_capsule
   center=<4.5339 -0.2502 0.3086>
   rotation=<rot -20.2902 -5.586 17.5303>
  }
  {
   shapeType=Capsule
   length=8.5084
   radius=1.2313
   shapeName=dyn_hairSide_l_03_capsule
   center=<1.1524 -0.1482 0.1236>
   rotation=<rot -6.9658 -8.0759 14.515>
  }
 ]
 dyn_hairSide_r_01 [{
   shapeType=Capsule
   length=6.8803
   radius=1.058
   shapeName=dyn_hairSide_l_01_capsule
   center=<2.0087 -1.4599 -0.0078>
   rotation=<rot 0 0 0.3997>
  }]
 dyn_hairSide_r_02 [
  {
   shapeType=Capsule
   length=7.344
   radius=1.1475
   shapeName=dyn_hairSide_l_02_capsule
   center=<2 -0.2156 0>
   rotation=<rot 0.0187 -0.0062 -3.6071>
  }
  {
   shapeType=Capsule
   length=7.9637
   radius=1.1475
   shapeName=dyn_hairSide_l_02_capsule
   center=<4.4713 -0.2588 0.0657>
   rotation=<rot 3.5482 -0.4679 -7.411>
  }
 ]
 dyn_hairSide_r_03 [
  {
   shapeType=Capsule
   length=9.2846
   radius=1.2313
   shapeName=dyn_hairSide_l_03_capsule
   center=<4.5339 -0.2502 0.3086>
   rotation=<rot 20.2902 -5.586 -17.5303>
  }
  {
   shapeType=Capsule
   length=8.5084
   radius=1.2313
   shapeName=dyn_hairSide_l_03_capsule
   center=<1.1524 -0.1482 0.1236>
   rotation=<rot 6.9658 -8.0759 -14.515>
  }
 ]
 dyn_hat_01_l [{shapeType=Capsule, length=11.3589, radius=2, shapeName=dyn_hat_01_r_capsule, center=<6.1262 0 0>, rotation=<rot 90 0 0>}]
 dyn_hat_01_r [{shapeType=Capsule, length=11.3589, radius=2, shapeName=dyn_hat_01_r_capsule, center=<6.1262 0 0>, rotation=<rot 90 0 0>}]
 dyn_hat_02_l [{shapeType=Capsule, length=7.9556, radius=2, shapeName=dyn_hat_02_r_capsule, center=<4.0367 0 0>, rotation=<rot 90 0 0>}]
 dyn_hat_02_r [{shapeType=Capsule, length=7.9556, radius=2, shapeName=dyn_hat_02_r_capsule, center=<4.0367 0 0>, rotation=<rot 90 0 0>}]
 dyn_hat_03_l [{shapeType=Capsule, length=6.516, radius=2, shapeName=dyn_hat_03_l_capsule, center=<4.0367 0 0>, rotation=<rot 90 0 0>}]
 dyn_hat_03_r [{shapeType=Capsule, length=6.516, radius=2, shapeName=dyn_hat_03_r_capsule, center=<4.0367 0 0>, rotation=<rot 90 0 0>}]
 head [{shapeType=Capsule, length=2.6794, radius=17.3374, shapeName=head_capsule, center=<6.9795 2.8293 0>, rotation=<rot 90 0 0>}]
}
constraints [
 {bone1Name=dyn_hairSide_l_01, bone2Name=base_hairSide_l_01, swing1LimitDegrees=20, swing2LimitDegrees=17, twistLimitDegrees=1}
 {bone1Name=dyn_hairSide_l_02, bone2Name=dyn_hairSide_l_01, swing1LimitDegrees=23, swing2LimitDegrees=15, twistLimitDegrees=5}
 {bone1Name=dyn_hairSide_l_03, bone2Name=dyn_hairSide_l_02, swing1LimitDegrees=25, swing2LimitDegrees=10, twistLimitDegrees=5}
 {
  bone1Name=base_hairSide_l_01
  bone2Name=head
  swing1LimitDegrees=45
  swing1Motion=Free
  swing2LimitDegrees=45
  swing2Motion=Free
  twistLimitDegrees=45
  twistMotion=Free
 }
 {
  bone1Name=base_hat_01_l
  bone2Name=head
  swing1LimitDegrees=45
  swing1Motion=Free
  swing2LimitDegrees=45
  swing2Motion=Free
  twistLimitDegrees=45
  twistMotion=Free
 }
 {bone1Name=dyn_hat_01_l, bone2Name=base_hat_01_l, swing1LimitDegrees=15, swing2LimitDegrees=5, twistLimitDegrees=1}
 {bone1Name=dyn_hat_02_l, bone2Name=dyn_hat_01_l, swing1LimitDegrees=25, swing2LimitDegrees=7, twistLimitDegrees=1}
 {bone1Name=dyn_hat_03_l, bone2Name=dyn_hat_02_l, swing1LimitDegrees=32, swing2LimitDegrees=10, twistLimitDegrees=1}
 {
  bone1Name=base_hat_01_r
  bone2Name=head
  swing1LimitDegrees=45
  swing1Motion=Free
  swing2LimitDegrees=45
  swing2Motion=Free
  twistLimitDegrees=45
  twistMotion=Free
 }
 {bone1Name=dyn_hat_01_r, bone2Name=base_hat_01_r, swing1LimitDegrees=15, swing2LimitDegrees=5, twistLimitDegrees=1}
 {bone1Name=dyn_hat_02_r, bone2Name=dyn_hat_01_r, swing1LimitDegrees=25, swing2LimitDegrees=7, twistLimitDegrees=1}
 {bone1Name=dyn_hat_03_r, bone2Name=dyn_hat_02_r, swing1LimitDegrees=32, swing2LimitDegrees=10, twistLimitDegrees=1}
 {
  bone1Name=base_hairSide_r_01
  bone2Name=head
  swing1LimitDegrees=45
  swing1Motion=Free
  swing2LimitDegrees=45
  swing2Motion=Free
  twistLimitDegrees=45
  twistMotion=Free
 }
 {bone1Name=dyn_hairSide_r_01, bone2Name=base_hairSide_r_01, swing1LimitDegrees=20, swing2LimitDegrees=17, twistLimitDegrees=1}
 {bone1Name=dyn_hairSide_r_02, bone2Name=dyn_hairSide_r_01, swing1LimitDegrees=20, swing2LimitDegrees=15, twistLimitDegrees=5}
 {bone1Name=dyn_hairSide_r_03, bone2Name=dyn_hairSide_r_02, swing1LimitDegrees=20, swing2LimitDegrees=10, twistLimitDegrees=5}
]
