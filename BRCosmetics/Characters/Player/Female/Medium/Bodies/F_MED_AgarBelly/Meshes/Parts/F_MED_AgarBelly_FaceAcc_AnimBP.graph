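# Post-process style Anim Blueprint for AgarBelly's face accessory mesh, adding secondary physics motion to head-attached bones.
# It copies the parent mesh pose, passes a disabled hat-bone tilt, then runs head-space RigidBody and two hair-bang AnimDynamics nodes.
# The event graph updates RigidBody gravity from a parameter asset and derives scaleDownAlpha from local-player aiming and weapon state.
# That flag drives only an unconnected tilt node; the hat AnimDynamics chain is also unconnected.
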AnimBlueprint /BRCosmetics/Characters/Player/Female/Medium/Bodies/F_MED_AgarBelly/Meshes/Parts/F_MED_AgarBelly_FaceAcc_AnimBP
target_skeleton=&/F_MED_AgarBelly_FaceAcc_Skeleton
graphs [
 AnimGraph {nodes [
   n1 = Anim Dynamics - Bone: dyn_hat_01_l(Alpha=1, ExternalForce="0,0,0") {
    angularDampingOverride=1
    angularSpringConstant=25000
    bAngularSpring=true
    bOverrideAngularDamping=true
    bUseGravityOverride=true
    gravityScale=0.35
    simulationSpace=RootRelative
    worldSpaceMinimumScale=0
    boundBone {boneName=dyn_hat_01_l}
    componentLinearAccScale=<1 1 1>
    componentLinearVelScale=<1 1 1>
    gravityOverride {z=-1100}
    physicsBodyDefinitions [{
      boundBone {boneName=dyn_hat_01_l}
      boxExtents=<50 50 50>
      constraintSetup {angularLimitsMax=<0 5 15>, angularLimitsMin=<0 -5 -15>, angularTarget=<1 0 0>}
      localJointOffset=<2.5 0 0>
     }]
    relativeSpaceBone {boneName=head}
    simSpaceSettings {maxAngularAcceleration=800, maxAngularVelocity=600, simSpaceAngularAlpha=1}
   }
   n2 = Anim Dynamics - Bone: dyn_hat_02_l(ComponentPose=@n1, Alpha=1, ExternalForce="0,0,0") {
    angularDampingOverride=1
    angularSpringConstant=20000
    bAngularSpring=true
    bOverrideAngularDamping=true
    bUseGravityOverride=true
    gravityScale=0.35
    simulationSpace=RootRelative
    worldSpaceMinimumScale=0
    boundBone {boneName=dyn_hat_02_l}
    componentLinearAccScale=<1 1 1>
    componentLinearVelScale=<1 1 1>
    gravityOverride {z=-1100}
    physicsBodyDefinitions [{
      boundBone {boneName=dyn_hat_02_l}
      boxExtents=<35 35 35>
      constraintSetup {angularLimitsMax=<0 7 20>, angularLimitsMin=<0 -7 -15>, angularTarget=<1 0 0>}
      localJointOffset=<3 0 0>
     }]
    relativeSpaceBone {boneName=head}
    simSpaceSettings {maxAngularAcceleration=800, maxAngularVelocity=600, simSpaceAngularAlpha=1}
   }
   n3 = Anim Dynamics - Bone: dyn_hat_03_l(ComponentPose=@n2, Alpha=1, ExternalForce="0,0,0") {
    angularDampingOverride=0.9999
    angularSpringConstant=10000
    bAngularSpring=true
    bOverrideAngularDamping=true
    bUseGravityOverride=true
    gravityScale=0.35
    simulationSpace=RootRelative
    worldSpaceMinimumScale=0
    boundBone {boneName=dyn_hat_03_l}
    componentLinearAccScale=<1 1 1>
    componentLinearVelScale=<1 1 1>
    gravityOverride {z=-1100}
    physicsBodyDefinitions [{
      boundBone {boneName=dyn_hat_03_l}
      boxExtents=<30 30 30>
      constraintSetup {angularLimitsMax=<0 10 20>, angularLimitsMin=<0 -10 -15>, angularTarget=<1 0 0>}
      localJointOffset=<3.5 0 0>
     }]
    relativeSpaceBone {boneName=head}
    simSpaceSettings {maxAngularAcceleration=800, maxAngularVelocity=600, simSpaceAngularAlpha=1}
   }
   n4 = Anim Dynamics - Bone: dyn_hat_01_r(ComponentPose=@n3, Alpha=1, ExternalForce="0,0,0") {
    angularDampingOverride=1
    angularSpringConstant=25000
    bAngularSpring=true
    bOverrideAngularDamping=true
    bUseGravityOverride=true
    gravityScale=0.35
    simulationSpace=RootRelative
    worldSpaceMinimumScale=0
    boundBone {boneName=dyn_hat_01_r}
    componentLinearAccScale=<1 1 1>
    componentLinearVelScale=<1 1 1>
    gravityOverride {z=-1100}
    physicsBodyDefinitions [{
      boundBone {boneName=dyn_hat_01_r}
      boxExtents=<50 50 50>
      constraintSetup {angularLimitsMax=<0 5 15>, angularLimitsMin=<0 -5 -15>, angularTarget=<1 0 0>}
      localJointOffset=<2.5 0 0>
     }]
    relativeSpaceBone {boneName=head}
    simSpaceSettings {maxAngularAcceleration=2800, maxAngularVelocity=600, simSpaceAngularAlpha=1}
   }
   n5 = Anim Dynamics - Bone: dyn_hat_02_r(ComponentPose=@n4, Alpha=1, ExternalForce="0,0,0") {
    angularDampingOverride=1
    angularSpringConstant=20000
    bAngularSpring=true
    bOverrideAngularDamping=true
    bUseGravityOverride=true
    gravityScale=0.35
    simulationSpace=RootRelative
    worldSpaceMinimumScale=0
    boundBone {boneName=dyn_hat_02_r}
    componentLinearAccScale=<1 1 1>
    componentLinearVelScale=<1 1 1>
    gravityOverride {z=-1100}
    physicsBodyDefinitions [{
      boundBone {boneName=dyn_hat_02_r}
      boxExtents=<35 35 35>
      constraintSetup {angularLimitsMax=<0 7 20>, angularLimitsMin=<0 -7 -15>, angularTarget=<1 0 0>}
      localJointOffset=<3 0 0>
     }]
    relativeSpaceBone {boneName=head}
    simSpaceSettings {maxAngularAcceleration=2800, maxAngularVelocity=600, simSpaceAngularAlpha=1}
   }
   Anim Dynamics - Bone: dyn_hat_03_r(ComponentPose=@n5, Alpha=1, ExternalForce="0,0,0") {
    angularDampingOverride=0.9999
    angularSpringConstant=10000
    bAngularSpring=true
    bOverrideAngularDamping=true
    bUseGravityOverride=true
    gravityScale=0.35
    simulationSpace=RootRelative
    worldSpaceMinimumScale=0
    boundBone {boneName=dyn_hat_03_r}
    componentLinearAccScale=<1 1 1>
    componentLinearVelScale=<1 1 1>
    gravityOverride {z=-1100}
    physicsBodyDefinitions [{
      boundBone {boneName=dyn_hat_03_r}
      boxExtents=<30 30 30>
      constraintSetup {angularLimitsMax=<0 10 20>, angularLimitsMin=<0 -10 -15>, angularTarget=<1 0 0>}
      localJointOffset=<3.5 0 0>
     }]
    relativeSpaceBone {boneName=head}
    simSpaceSettings {maxAngularAcceleration=800, maxAngularVelocity=600, simSpaceAngularAlpha=1}
   }
   "Transform (Modify) Bone - Bone: base_hat_01_r"(Alpha="To Float (Boolean)"(InBool=Get scaleDownAlpha() -> (scaleDownAlpha:bool)), Translation="0,0,0", Rotation="20,0,0", Scale="1,1,1") {
    rotationMode=BMM_Additive
    rotationSpace=BCS_BoneSpace
    scaleMode=BMM_Replace
    scaleSpace=BCS_BoneSpace
    alphaBoolBlend {blendInTime=0.1, blendOption=Linear, blendOutTime=0.12}
    alphaScaleBiasClamp {bInterpResult=true}
    boneToModify {boneName=base_hat_01_r}
    rotation {pitch=20}
   }
   RigidBody!(ComponentPose, Alpha, OverrideWorldGravity:Vector="0,0,-980", ExternalForce) -> (Pose) {
    bClampLinearTranslationLimitToRefPose=true
    bOverrideWorldGravity=true
    overridePhysicsAsset=&/F_MED_AgarBelly_FaceAcc_Physics
    simulationSpace=BaseBoneSpace
    baseBoneRef {boneName=head}
    simSpaceSettings {
     dampingAlpha=0
     maxAngularAcceleration=12
     maxAngularVelocity=6
     maxLinearAcceleration=900
     maxLinearVelocity=700
     worldAlpha=1
     externalLinearDragV=<0.2 0.2 0.2>
    }
   }
   n6 = "Transform (Modify) Bone - Bone: base_hat_01_r"!(
    ComponentPose=Local To Component(LocalPose=Copy Pose From Mesh() {bUseAttachedParent=true})
    bAlphaBoolEnabled:bool=false
    Translation="0,0,0"
    Rotation="25,0,0"
    Scale="1,1,1"
   ) -> (Pose) {
    alphaInputType=Bool
    rotationMode=BMM_Additive
    rotationSpace=BCS_BoneSpace
    scaleMode=BMM_Replace
    scaleSpace=BCS_BoneSpace
    alphaBoolBlend {blendInTime=0.2, blendOption=Sinusoidal, blendOutTime=0.2}
    alphaScaleBiasClamp {bInterpResult=true}
    boneToModify {boneName=base_hat_01_r}
    rotation {pitch=25}
   }
   n7 = RigidBody(ComponentPose=@n6, OverrideWorldGravity=Get GravityOverride Params() -> (GravityOverride Params:Vector)) {
    bClampLinearTranslationLimitToRefPose=true
    bOverrideWorldGravity=true
    overridePhysicsAsset=&F_MED_AgarBelly_FaceAcc_Physics
    simulationSpace=BaseBoneSpace
    baseBoneRef {boneName=head}
    simSpaceSettings {
     dampingAlpha=0
     maxAngularAcceleration=12
     maxAngularVelocity=6
     maxLinearAcceleration=900
     maxLinearVelocity=700
     worldAlpha=1
     externalLinearDragV=<0.2 0.2 0.2>
    }
   }
   n8 = Anim Dynamics - Bone: dyn_hairBang_1_01(ComponentPose=@n7, Alpha=1, ExternalForce="0,0,0") {
    angularDampingOverride=1
    angularSpringConstant=15000
    bAngularSpring=true
    bOverrideAngularDamping=true
    bUseGravityOverride=true
    gravityScale=0.35
    simulationSpace=RootRelative
    worldSpaceMinimumScale=0
    boundBone {boneName=dyn_hairBang_1_01}
    componentLinearAccScale=<1 1 1>
    componentLinearVelScale=<1 1 1>
    gravityOverride {z=-980}
    physicsBodyDefinitions [{
      boundBone {boneName=dyn_hairBang_1_01}
      boxExtents=<35 35 35>
      constraintSetup {angularLimitsMax=<0 2 17>, angularLimitsMin=<0 -2 -15>, angularTarget=<1 0 0>}
      localJointOffset=<2 0 0>
     }]
    simSpaceSettings {maxAngularAcceleration=2800, maxAngularVelocity=700, simSpaceAngularAlpha=1}
   }
   n9 = Anim Dynamics - Bone: dyn_hairBang_2_01(ComponentPose=@n8, Alpha=1, ExternalForce="0,0,0") {
    angularDampingOverride=1
    angularSpringConstant=15000
    bAngularSpring=true
    bOverrideAngularDamping=true
    bUseGravityOverride=true
    gravityScale=0.35
    simulationSpace=RootRelative
    worldSpaceMinimumScale=0
    boundBone {boneName=dyn_hairBang_2_01}
    componentLinearAccScale=<1 1 1>
    componentLinearVelScale=<1 1 1>
    gravityOverride {z=-980}
    physicsBodyDefinitions [{
      boundBone {boneName=dyn_hairBang_2_01}
      boxExtents=<40 40 40>
      constraintSetup {angularLimitsMax=<0 2 25>, angularLimitsMin=<0 -2 -15>, angularTarget=<1 0 0>}
      localJointOffset=<2 0 0>
     }]
    simSpaceSettings {maxAngularAcceleration=2800, maxAngularVelocity=700, simSpaceAngularAlpha=1}
   }
   Output Pose(Result=Component To Local(ComponentPose=@n9)) {layerGroup=None}
  ]}
 EventGraph {nodes [
   Event BlueprintUpdateAnimation()
   > n1 = Sequence!(execute) -> (then_0, then_1, then_2:Exec)
   > n2 = "Gravity Override State Function (Parameter Asset) (Nativized)"(
    CharacterPartAnimInstance=Self-Reference()
    GameTimeInSeconds=GetGameTimeInSeconds()
    Parameters=&/F_MED_AgarBelly_FaceAcc_Parameters
   )
   > Set GravityOverride Params(GravityOverride Params:Vector=@n2) -> (Output_Get:Vector)
   > Set RBAN Settings Out(RBAN Settings Out=@n2.OutRigidBodyAnimNodeParameters)
   Set FortPawn(execute=@n1.then_1, FortPawn:Fort Pawn Object Reference=TryGetFortPawn()) -> (Output_Get:Fort Pawn Object Reference)
   n3 = Get FortPawn() -> (FortPawn:Fort Pawn Object Reference)
   > n4 = Is Valid(InputObject=@n3)
   Set weapon_visible(execute=@n4.Is Not Valid, weapon_visible:bool=false) -> (Output_Get:bool)
   n5 = Get CurrentWeapon!(self:Fort Pawn Object Reference=@n3) -> (CurrentWeapon:Fort Weapon Object Reference)
   n6 = GetCoreAnimation!(self:Fort Weapon Object Reference=@n5) -> (ReturnValue:EFortWeaponCoreAnimation Enum)
   n7 = Set notCoverCrosshair(
    execute=@n4.Is Valid
    notCoverCrosshair:bool=OR Boolean!(
     A="Equal (Enum)"(A:EFortWeaponCoreAnimation Enum=@n6, B:EFortWeaponCoreAnimation Enum=TrapPlacement)
     B="Equal (Enum)"(A=@n6, B=RocketLauncher)
     C:bool="Equal (Enum)"(A=@n6, B=GoingCommando)
    ) -> (ReturnValue)
   ) -> (Output_Get:bool)
   > n8 = Branch(Condition=@n7)
   > Set weapon_visible(weapon_visible=false)
   n9 = IsWeaponHolstered!(self:Fort Pawn Object Reference=@n3) -> (ReturnValue:bool)
   Set weapon_visible(execute=@n8.else, weapon_visible=AND Boolean(A=IsValid(Object=@n5), B=NOT Boolean(A=@n9)))
   n10 = Set isLocalPlayer(execute=@n1.then_2, isLocalPlayer:bool=IsLocalPawn!(Actor:Actor Object Reference=GetOwningActor()) -> (IsLocalPawn:bool)) -> (Output_Get:bool)
   > n11 = Branch(Condition=AND Boolean!(A=@n10, B=Get bIsTargeting(), C:bool=Get weapon_visible() -> (weapon_visible:bool)) -> (ReturnValue))
   > Set scaleDownAlpha(scaleDownAlpha:bool=true) -> (Output_Get:bool)
   Set scaleDownAlpha(execute=@n11.else, scaleDownAlpha=false)
  ]}
]
variables [
 FortPawn {type=FortPawn, default=null}
 GravityOverride Params {type=Vector}
 isLocalPlayer {type=boolean}
 notCoverCrosshair {type=boolean}
 RBAN Settings Out {type=FortRigidBodyAnimNodeParameters}
 scaleDownAlpha {type=boolean}
 weapon_visible {type=boolean}
]
